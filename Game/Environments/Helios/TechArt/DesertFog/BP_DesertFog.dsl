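; BP_DesertFog is a desert fog Blueprint that sizes a local fog volume from a radius value.
; The custom PostSpawn event sets the LocalFogVolume world scale uniformly to the Radius variable.
; It then prints the radius as a debug string; the construction script is empty.
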(event Custom|PostSpawn
  (bind _radius (Variables|Default|GetRadius))
  (Transformation|SetWorldScale3D (Variables|Default|GetLocalFogVolume) (Math|Conversions|ToVector(Float) _radius))
  (Development|PrintString (Utilities|String|Append "PostSpawn
Radius: " (Utilities|String|ToString(Float) _radius))))

(fn ConstructionScript)
